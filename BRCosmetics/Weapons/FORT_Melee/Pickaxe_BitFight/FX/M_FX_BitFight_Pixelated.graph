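# Translucent unlit Niagara sprite material for the BitFight pickaxe FX that renders an impact texture as a grid of pixels.
# UVs are quantized per tile to sample the texture, then masked by hard-edged squares that leave gaps between pixels.
# Particle alpha biases and scales that mask for a dissolve-style fade, with optional depth fade.
# Emissive starts as boosted particle color and settles to plain color over particle lifetime, with mobile skipping the boost.
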Material /BRCosmetics/Weapons/FORT_Melee/Pickaxe_BitFight/FX/M_FX_BitFight_Pixelated
graph {
 properties {bUsedWithNiagaraSprites=true, blendMode=BLEND_Translucent, pixelDepthOffsetMode=PDOM_Legacy, shadingModel=MSM_Unlit, usedShadingModels=Unlit}
 nodes [
  n1 = ParticleColor()
  n2 = Multiply(A=ScalarParameter() {defaultValue=4, parameterName="Particle color start  Multiply"}, B=@n1.RGB) {constA=4}
  n3 = ParticleRelativeTime()
  MaterialFunctionCall(Emissive=ShadingPathSwitch(Default=LinearInterpolate(A=@n2, B=@n1.RGB, Alpha=@n3) {constA=6}, Mobile=@n1.RGB)) {materialFunction=&MF_FX_EmissiveAdjustments}
  n4 = ScalarParameter() {defaultValue=0.8, group=Base Pixels, parameterName=Pixel Size, sortPriority=2}
  n5 = AppendVector(
   A=ScalarParameter() {defaultValue=64, group=Base Pixels, parameterName=Pixel Tile X, sortPriority=0}
   B=ScalarParameter() {defaultValue=64, group=Base Pixels, parameterName=Pixel Tile Y, sortPriority=1}
  )
  n6 = Multiply(A=TextureCoordinate(), B=@n5)
  n7 = Frac(None=@n6)
  n8 = MaterialFunctionCall(Float2=Subtract(A=Multiply(A=@n7) {constB=2})) {materialFunction=&BreakOutFloat2Components}
  n9 = OneMinus(None=SmoothStep(Min=@n4, Max=@n4, Value=@n8.R))
  n10 = OneMinus(None=SmoothStep(Min=@n4, Max=@n4, Value=@n8.G))
  n11 = Divide(B=@n5)
  n12 = Multiply(A=Floor(None=@n6), B=@n11)
  n13 = TextureSampleParameter2D!(
   UVs=Multiply(A=Add(A=@n12, B=@n11), B=ScalarParameter() {defaultValue=1, group=Base Pixels, parameterName=Pixel Texture Scale})
   Level
   Apply View MipBias
  ) {mipValueMode=TMVM_MipLevel, parameterName=ImpactTexture, texture=/Game/Effects/Fort_Effects/Textures/Explo/T_Impact_Flash_2x2_01}
  n14 = Multiply(A=Saturate(None=Multiply(A=@n9, B=@n10)), B=@n13.RGB)
  n15 = Add(A=LinearInterpolate(Alpha=@n1.A) {constA=-0.6, constB=0.6}, B=@n14)
  n16 = Multiply(A=@n1.A, B=Clamp(None=Multiply(A=@n15, B=@n14)))
  n17 = ScalarParameter() {defaultValue=25, parameterName=DepthFade}
  Multiply(A=LinearInterpolate(A=@n16, B=DepthFade(Opacity=@n16, FadeDistance=@n17) {fadeDistanceDefault=25}, Alpha=ScalarParameter() {parameterName=UseDepthFade}))
 ]
}
