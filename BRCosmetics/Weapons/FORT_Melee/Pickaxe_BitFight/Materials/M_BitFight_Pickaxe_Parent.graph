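# Parent material for the BitFight pickaxe, adding an animated emissive screen arrow between the shared 2023 character PreFX and PostFX functions.
# A SwingCurve atlas row, sampled on looping time, feeds named reroutes that nudge the arrow texture's offset, scale and intensity.
# A WeakSpotHit timestamp and duration form a timed WeakSpotActive switch, apparently driving blends to a red arrow and a continuous rather than intermittent pulse.
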Material /BRCosmetics/Weapons/FORT_Melee/Pickaxe_BitFight/Materials/M_BitFight_Pickaxe_Parent
graph {
 properties {
  bHasPixelAnimation=true
  bUseMaterialAttributes=true
  bUsedWithMeshDeformer=true
  bUsedWithMorphTargets=true
  bUsedWithSkeletalMesh=true
  pixelDepthOffsetMode=PDOM_Legacy
  shadingModel=MSM_FromMaterialExpression
  usedShadingModels=Default Lit | Subsurface | Clear Coat | Subsurface Profile | Cloth
 }
 nodes [
  n1 = MaterialFunctionCall() {materialFunction=&MF_CharacterParent_2023_PreFX}
  n2 = GetMaterialAttributes(None=@n1.Result) {attributeGetTypes [#1]}
  n3 = TextureCoordinate()
  n4 = AppendVector(A=ScalarParameter() {parameterName=Arrow U Offset}, B=Subtract(A=ScalarParameter() {parameterName=Arrow V Offset}, B=NamedRerouteUsage()))
  n5 = ScalarParameter() {defaultValue=1, parameterName=Arrow Scale}
  n6 = NamedRerouteUsage()
  n7 = MaterialFunctionCall(
   UVs=MaterialFunctionCall(UVs=Add(A=@n3, B=@n4), Texture Scale=Add(A=@n5, B=@n6)) {materialFunction=&ScaleUVsByCenter}.UVs
   "Rotation Angle (0-1)"=Constant() {r=0.5}
  ) {materialFunction=&CustomRotator}
  n8 = Multiply(
   A=TextureSample(UVs=Saturate(None=@n7.Rotated Values)) {texture=/BRCosmetics/Weapons/FORT_Melee/Pickaxe_BitFight/Textures/T_BitFight_ScreenAlpha}.RGB
   B=Add(A=ScalarParameter() {defaultValue=1, parameterName=Arrow Intensity}, B=NamedRerouteUsage())
  )
  n9 = Constant() {r=1}
  n10 = LinearInterpolate(A=@n8, B=Multiply(A=Desaturation(None=@n8, Fraction=@n9), B=Constant3Vector() {constant=<rgba 3 0 0 0>}), Alpha=NamedRerouteUsage())
  MaterialFunctionCall(
   Material=SetMaterialAttributes(MaterialAttributes=@n2.MaterialAttributes, Emissive Color=Add(A=@n2.Emissive Color, B=@n10)) {attributeSetTypes [#1]}
   M=@n1.M
   Pre Skinned Local Position=@n1.M
   Distance Blend=@n1.M
   Fresnel=@n1.M
  ) {materialFunction=&MF_CharacterParent_2023_PostFX}
  n11 = Time()
  n12 = Frac(None=@n11)
  n13 = ScalarParameter() {defaultValue=0.25, parameterName=LoopSpeed}
  n14 = Constant()
  n15 = Frac(None=Divide(A=@n11, B=@n13))
  n16 = Divide(A=Time()) {constB=0.25}
  n17 = CurveAtlasRowParameter(CurveTime=LinearInterpolate(A=If(A=@n12, B=@n13, "A > B"=@n14, "A == B"=@n15, "A < B"=@n15), B=Frac(None=@n16), Alpha=NamedRerouteUsage())) {
   atlas=&/A_Pickaxe_BitFight_SwingCurve
   curve=&/C_Pickaxe_BitFight_SwingCurve
   parameterName=SwingCurve
  }
  NamedRerouteDeclaration(None=Multiply(A=@n17.R) {constB=0.01}) {name=Offset, nodeColor=<rgba 0 1 0.4141 1>}
  NamedRerouteDeclaration(None=Multiply(A=@n17.R) {constB=1.5}) {name=Intensity, nodeColor=<rgba 1 0.0938 0 1>}
  NamedRerouteDeclaration(None=Multiply(A=@n17.R) {constB=0.005}) {name=Scale, nodeColor=<rgba 0 1 0.6484 1>}
  n18 = Time()
  n19 = ScalarParameter() {parameterName=WeakSpotHit}
  n20 = Saturate(None=Divide(A=Subtract(A=@n18, B=@n19), B=ScalarParameter() {defaultValue=2.5, parameterName=WeakSpot Duration}))
  NamedRerouteDeclaration(None=OneMinus(None=Floor(None=@n20))) {name=WeakSpotActive, nodeColor=<rgba 0.1719 1 0 1>}
 ]
}
